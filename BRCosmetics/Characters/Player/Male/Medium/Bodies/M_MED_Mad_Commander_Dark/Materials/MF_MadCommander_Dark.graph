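# Material function for the Mad Commander Dark character that animates emissive color, adding moving rings to the eyes and cycling hue over time.
# UVs are remapped into the right-eye crop box, with an offset repeat for the left eye, then a time-scrolled radial distance samples a curve atlas.
# A hard circular mask blends the rings into incoming emissive under the IsHead switch.
# HueShift then shifts the emissive result on head and non-head alike.
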MaterialFunction /BRCosmetics/Characters/Player/Male/Medium/Bodies/M_MED_Mad_Commander_Dark/Materials/MF_MadCommander_Dark
graph {nodes [
  n1 = GetMaterialAttributes(None=FunctionInput() {inputType=FunctionInput_MaterialAttributes}) {attributeGetTypes [#1]}
  n2 = Time()
  n3 = ScalarParameter() {defaultValue=0.3, group=MadCommander, parameterName=MC_HueShiftSpeed, sliderMax=1}
  n4 = Frac(None=Multiply(A=@n2, B=Multiply(A=@n3)) {constB=0.2})
  n5 = TextureCoordinate()
  n6 = ComponentMask(None=VectorParameter() {group=MadCommander, parameterName=MC_RightEyeCrop-LowerRight, defaultValue=<rgba 0.346 0.33 0 1>}.RGB) {g=true, r=true}
  n7 = Multiply(A=Constant() {r=-1}, B=@n6)
  n8 = ComponentMask(None=VectorParameter() {group=MadCommander, parameterName=MC_RightEyeCrop-UpperLeft, defaultValue=<rgba 0.3 0.415 0 1>}.RGB) {g=true, r=true}
  n9 = MaterialFunctionCall(Float2=Divide(A=Add(A=@n5, B=@n7), B=Subtract(A=@n8, B=@n6))) {materialFunction=&BreakOutFloat2Components}
  n10 = ScalarParameter() {defaultValue=2, group=MadCommander, parameterName="MC_LeftEye (Right EyeOffset)"}
  n11 = Subtract(A=@n9.R, B=@n10)
  n12 = AppendVector(A=If(A=@n9.R, B=@n10, "A > B"=@n11, "A == B"=@n11, "A < B"=@n9.R), B=@n9.G)
  n13 = Constant() {r=0.5}
  n14 = Multiply(A=Distance(A=@n12, B=@n13)) {constB=2.2}
  n15 = Frac(None=Multiply(A=@n2, B=ScalarParameter() {defaultValue=0.3, group=MadCommander, parameterName=MC_RingOffsetsSpeed, sliderMax=1}) {constB=0.2})
  n16 = CurveAtlasRowParameter(CurveTime=Frac(None=Add(A=@n14, B=@n15))) {
   atlas=&AthenaCurveAtlas01
   curve=/Game/Characters/Player/Male/Medium/Bodies/M_MED_Mad_Commander_Dark/Materials/Curve_MadCommander
   defaultValue=255
   group=MadCommander
   parameterName=Curve
  }
  n17 = SphereMask(A=@n12, B=@n13, Radius=@n13, Hardness=Constant() {r=1})
  n18 = StaticSwitchParameter(True=LinearInterpolate(A=@n1.EmissiveColor, B=@n16, Alpha=Ceil(None=@n17)), False=@n1.EmissiveColor) {group=MadCommander, parameterName=IsHead}
  FunctionOutput(None=SetMaterialAttributes!(
    MaterialAttributes=@n1.MaterialAttributes
    EmissiveColor=MaterialFunctionCall(Hue Shift Percentage=@n4, Texture=@n18) {materialFunction=&HueShift}.Result
   ) {attributeSetTypes [#1]})
  Power(Base=@n17)
  Sine(None=@n4)
 ]}
