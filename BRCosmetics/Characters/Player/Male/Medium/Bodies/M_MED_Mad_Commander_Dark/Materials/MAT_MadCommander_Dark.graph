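# Character material for the MadCommander Dark medium male body cosmetic, built on material attributes for skeletal meshes with morph targets and mesh deformer support.
# The shared PreFX character function feeds its result into the skin-specific MF_MadCommander_Dark function, whose output goes to the shared PostFX function.
# PreFX's AO output also drives the AO, Distance Blend and Fresnel inputs of PostFX.
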Material /BRCosmetics/Characters/Player/Male/Medium/Bodies/M_MED_Mad_Commander_Dark/Materials/MAT_MadCommander_Dark
graph {
 properties {
  bUseMaterialAttributes=true
  bUsedWithMeshDeformer=true
  bUsedWithMorphTargets=true
  bUsedWithSkeletalMesh=true
  numCustomizedUVs=2
  pixelDepthOffsetMode=PDOM_Legacy
  usedShadingModels=Default Lit
 }
 nodes [
  n1 = MaterialFunctionCall() {materialFunction=&BaseFN_CharacterFunction_PreFX}
  MaterialFunctionCall(
   AO=@n1.AO
   Distance Blend=@n1.AO
   Fresnel=@n1.AO
   Material=MaterialFunctionCall!(In=@n1.Result) {materialFunction=&/MF_MadCommander_Dark}.Result
  ) {materialFunction=&BaseFN_CharacterFunction_PostFX}
 ]
}
