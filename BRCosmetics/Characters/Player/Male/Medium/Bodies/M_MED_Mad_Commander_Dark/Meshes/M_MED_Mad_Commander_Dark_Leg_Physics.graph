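# Physics asset for the Mad Commander Dark medium male body's leg mesh, set up for a dangling pouch on the left thigh.
# Three capsule bodies cover the pouch bone, the left thigh twist bone and a tiny pelvis body.
# The thigh is constrained to the pelvis with free swing and twist, while the pouch hangs off the thigh with narrow angular limits.
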PhysicsAsset /BRCosmetics/Characters/Player/Male/Medium/Bodies/M_MED_Mad_Commander_Dark/Meshes/M_MED_Mad_Commander_Dark_Leg_Physics
bodies {
 dyn_leg_l_pouch [{shapeType=Capsule, length=7.6294, radius=5.2054, center=<2.9219 2.0868 -0.0965>, rotation=<rot 0.1257 -179.4897 -88.6831>}]
 pelvis [{shapeType=Capsule, length=0.1, radius=0.1}]
 thigh_twist_01_l [{shapeType=Capsule, length=9.7239, radius=2.8229, rotation=<rot 90 0 0>}]
}
constraints [
 {
  bone1Name=thigh_twist_01_l
  bone2Name=pelvis
  swing1LimitDegrees=45
  swing1Motion=Free
  swing2LimitDegrees=45
  swing2Motion=Free
  twistLimitDegrees=45
  twistMotion=Free
 }
 {bone1Name=dyn_leg_l_pouch, bone2Name=thigh_twist_01_l, swing1LimitDegrees=5, swing2LimitDegrees=10, twistLimitDegrees=10}
]
