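# Physics asset for the Mad Commander Dark male medium body, simulating dangling grenades and hip pouches on both sides.
# Each dynamic bone joins a tiny MSTR anchor body through a constraint with narrow swing and twist limits.
# The anchors, spine_05 and right thigh link to the pelvis with free angular motion.
# A capsule on the right thigh twist bone is the only sizeable leg shape, likely a collider for the dangling gear.
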PhysicsAsset /BRCosmetics/Characters/Player/Male/Medium/Bodies/M_MED_Mad_Commander_Dark/Meshes/M_MED_Mad_Commander_Dark_Physics
bodies {
 dyn_grenade_l [{shapeType=Capsule, length=5.2738, radius=3.3863, center=<3.869 -0.2511 0.1404>, rotation=<rot 0 -11.7325 -57.0884>}]
 dyn_grenade_l_MSTR [{shapeType=Capsule, length=0.1, radius=0.1}]
 dyn_grenade_r [{shapeType=Capsule, length=4.0437, radius=3.6611, center=<2.5804 -0.2815 1.5653>, rotation=<rot 0 -11.7285 -57.058>}]
 dyn_grenade_r_MSTR [{shapeType=Capsule, length=0.1, radius=0.1}]
 dyn_hip_pouch_l [{shapeType=Capsule, length=4.0172, radius=4.7643, center=<2.7841 1.4186 0.7352>}]
 dyn_hip_pouch_l_MSTR [{shapeType=Capsule, length=0.1, radius=0.1}]
 dyn_hip_pouch_r [{shapeType=Box, extentX=5.9404, extentY=12.7095, extentZ=19.7605, center=<3.1074 3.3026 -0.0132>, rotation=<rot 0 170.1804 -87.7665>}]
 dyn_hip_pouch_r_MSTR [{shapeType=Capsule, length=0.1, radius=0.1}]
 pelvis [{shapeType=Capsule, length=0.1, radius=0.1}]
 spine_05 [{shapeType=Capsule, length=0.1, radius=0.1}]
 thigh_r [{shapeType=Capsule, length=0.1, radius=0.1}]
 thigh_twist_01_r [{shapeType=Capsule, length=9.9305, radius=5.94, center=<-15.2615 -2.2282 1.983>, rotation=<rot 75.7015 -117.4207 -31.2706>}]
}
constraints [
 {
  bone1Name=spine_05
  bone2Name=pelvis
  swing1LimitDegrees=45
  swing1Motion=Free
  swing2LimitDegrees=45
  swing2Motion=Free
  twistLimitDegrees=45
  twistMotion=Free
 }
 {
  bone1Name=dyn_grenade_l_MSTR
  bone2Name=pelvis
  swing1LimitDegrees=45
  swing1Motion=Free
  swing2LimitDegrees=45
  swing2Motion=Free
  twistLimitDegrees=45
  twistMotion=Free
 }
 {bone1Name=dyn_grenade_l, bone2Name=dyn_grenade_l_MSTR, swing1LimitDegrees=30, swing2LimitDegrees=30, twistLimitDegrees=10}
 {
  bone1Name=dyn_grenade_r_MSTR
  bone2Name=pelvis
  swing1LimitDegrees=45
  swing1Motion=Free
  swing2LimitDegrees=45
  swing2Motion=Free
  twistLimitDegrees=45
  twistMotion=Free
 }
 {bone1Name=dyn_grenade_r, bone2Name=dyn_grenade_r_MSTR, swing1LimitDegrees=30, swing2LimitDegrees=30, twistLimitDegrees=10}
 {
  bone1Name=dyn_hip_pouch_r_MSTR
  bone2Name=pelvis
  swing1LimitDegrees=45
  swing1Motion=Free
  swing2LimitDegrees=45
  swing2Motion=Free
  twistLimitDegrees=45
  twistMotion=Free
 }
 {bone1Name=dyn_hip_pouch_r, bone2Name=dyn_hip_pouch_r_MSTR, swing1LimitDegrees=25, swing2LimitDegrees=25, twistLimitDegrees=5}
 {
  bone1Name=dyn_hip_pouch_l_MSTR
  bone2Name=pelvis
  swing1LimitDegrees=45
  swing1Motion=Free
  swing2LimitDegrees=45
  swing2Motion=Free
  twistLimitDegrees=45
  twistMotion=Free
 }
 {bone1Name=dyn_hip_pouch_l, bone2Name=dyn_hip_pouch_l_MSTR, swing1LimitDegrees=20, swing2LimitDegrees=20, twistLimitDegrees=5}
 {
  bone1Name=thigh_r
  bone2Name=pelvis
  swing1LimitDegrees=45
  swing1Motion=Free
  swing2LimitDegrees=45
  swing2Motion=Free
  twistLimitDegrees=45
  twistMotion=Free
 }
 {
  bone1Name=thigh_twist_01_r
  bone2Name=thigh_r
  swing1LimitDegrees=45
  swing1Motion=Free
  swing2LimitDegrees=45
  swing2Motion=Free
  twistLimitDegrees=45
  twistMotion=Free
 }
]
