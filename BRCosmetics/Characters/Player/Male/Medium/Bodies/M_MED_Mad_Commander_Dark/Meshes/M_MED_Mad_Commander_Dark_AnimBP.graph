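# Post-copy cosmetic AnimBP for the Mad Commander Dark body, copying the parent mesh pose and running it through a Control Rig.
# Constraints rotate shoulder pads with upper-arm twist bones by 2D pawn speed, zeroed when skydiving or in vehicles, and knee pads between thigh and calf.
# Additive bone offsets adjust shoulder pads during skydiving and the right hip pouch when crouching.
# Two RigidBody passes then simulate secondary physics, relative to root and left thigh.
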AnimBlueprint /BRCosmetics/Characters/Player/Male/Medium/Bodies/M_MED_Mad_Commander_Dark/Meshes/M_MED_Mad_Commander_Dark_AnimBP
target_skeleton=/Game/Characters/Player/Male/Medium/Bodies/M_MED_Mad_Commander_Dark/Meshes/M_MED_Mad_Commander_Dark_Skeleton
graphs [
 AnimGraph {nodes [
   n1 = Control Rig(Source=Copy Pose From Mesh() {bUseAttachedParent=true}) {
    alphaBoolBlend {blendOption=Linear}
    controlRigAssetReference {blueprintRigClass=/Game/Characters/Player/Male/Medium/Bodies/M_MED_Mad_Commander_Dark/Meshes/M_MED_Mad_Commander_Dark_CtrlRig/M_MED_Mad_Commander_Dark_CtrlRig.M_MED_Mad_Commander_Dark_CtrlRig_C}
   }
   n2 = Get clampShldrPads() -> (clampShldrPads:double)
   n3 = "Constraint - shldr_pad_l (R)"(ComponentPose=Local To Component(LocalPose=@n1), Alpha=@n2, ConstraintWeights_0=1) {
    alphaBoolBlend {blendOption=Linear}
    alphaScaleBiasClamp {bMapRange=true, inRange {max=550, min=400}}
    boneToModify {boneName=shldr_pad_l}
    constraintSetup [{transformType=Rotation, targetBone {boneName=upperarm_twist_01_l}}]
    constraintWeights [1]
   }
   n4 = "Constraint - shldr_pad_r (R)"(ComponentPose=@n3, Alpha=@n2, ConstraintWeights_0=1) {
    alphaBoolBlend {blendOption=Linear}
    alphaScaleBiasClamp {bMapRange=true, inRange {max=550, min=400}}
    boneToModify {boneName=shldr_pad_r}
    constraintSetup [{transformType=Rotation, targetBone {boneName=upperarm_twist_01_r}}]
    constraintWeights [1]
   }
   n5 = "Constraint - knee_pad_l (RR)"!(ComponentPose=@n4, Alpha=1, ConstraintWeights_0=0.5, ConstraintWeights_1:float=0.5) -> (Pose) {
    alphaBoolBlend {blendOption=Linear}
    boneToModify {boneName=knee_pad_l}
    constraintSetup [{transformType=Rotation, targetBone {boneName=thigh_l}}, {transformType=Rotation, targetBone {boneName=calf_l}}]
    constraintWeights [1,1]
   }
   n6 = "Constraint - knee_pad_r (RR)"!(ComponentPose=@n5, Alpha=1, ConstraintWeights_0=0.5, ConstraintWeights_1:float=0.5) -> (Pose) {
    alphaBoolBlend {blendOption=Linear}
    boneToModify {boneName=knee_pad_r}
    constraintSetup [{transformType=Rotation, targetBone {boneName=thigh_r}}, {transformType=Rotation, targetBone {boneName=calf_r}}]
    constraintWeights [1,1]
   }
   n7 = Get bIsSkyDiving()
   n8 = "Transform (Modify) Bone - Bone: shldr_pad_l"!(ComponentPose=@n6, bAlphaBoolEnabled:bool=@n7, Translation="3,0,3", Rotation="0,0,0", Scale="1,1,1") -> (Pose) {
    alphaInputType=Bool
    translationMode=BMM_Additive
    translationSpace=BCS_BoneSpace
    alphaBoolBlend {blendOption=Linear}
    boneToModify {boneName=shldr_pad_l}
    translation {x=3, z=3}
   }
   n9 = "Transform (Modify) Bone - Bone: shldr_pad_r"!(ComponentPose=@n8, bAlphaBoolEnabled:bool=@n7, Translation="-3,0,-3", Rotation="0,0,0", Scale="1,1,1") -> (Pose) {
    alphaInputType=Bool
    translationMode=BMM_Additive
    translationSpace=BCS_BoneSpace
    alphaBoolBlend {blendOption=Linear}
    boneToModify {boneName=shldr_pad_r}
    translation {x=-3, z=-3}
   }
   n10 = "Transform (Modify) Bone - Bone: dyn_hip_pouch_r_MSTR"!(ComponentPose=@n9, bAlphaBoolEnabled:bool=Get bIsCrouching(), Translation="1.008,0.018,0.340", Rotation="0.000,20,-10.002", Scale="1,1,1") -> (Pose) {
    alphaInputType=Bool
    rotationMode=BMM_Additive
    translationMode=BMM_Additive
    translationSpace=BCS_BoneSpace
    alphaBoolBlend {blendOption=Linear}
    boneToModify {boneName=dyn_hip_pouch_r_MSTR}
    rotation {roll=-10.002, yaw=20}
    translation=<1.008 0.018 0.34>
   }
   n11 = RigidBody(ComponentPose=@n10) {
    bOverrideWorldGravity=true
    overridePhysicsAsset=&/M_MED_Mad_Commander_Dark_Physics
    simulationSpace=BaseBoneSpace
    alphaBoolBlend {blendOption=Linear}
    baseBoneRef {boneName=root}
    componentLinearAccScale=<0.3 0.2 0.2>
    componentLinearVelScale=<0.2 0.2 0.2>
    overrideWorldGravity {z=-980}
   }
   n12 = RigidBody(ComponentPose=@n11) {
    bOverrideWorldGravity=true
    overridePhysicsAsset=&/M_MED_Mad_Commander_Dark_Leg_Physics
    simulationSpace=BaseBoneSpace
    alphaBoolBlend {blendOption=Linear}
    baseBoneRef {boneName=thigh_l}
    componentLinearAccScale=<0.3 0.2 0.2>
    componentLinearVelScale=<0.2 0.2 0.2>
    overrideWorldGravity {z=-980}
   }
   Output Pose(Result=Component To Local(ComponentPose=@n12)) {layerGroup=None}
  ]}
 EventGraph {nodes [
   Event BlueprintUpdateAnimation()
   > n1 = Branch(Condition=Get bIsSkyDiving())
   Set clampShldrPads(execute=[@n1.then, @n2.then], clampShldrPads:double=0) -> (Output_Get:double)
   n2 = Branch(execute=@n1.else, Condition=Get bIsInVehicle() -> (bIsInVehicle:bool))
   Set clampShldrPads(execute=@n2.else, clampShldrPads=Get PawnSpeed2D())
  ]}
]
variables [clampShldrPads {type=number}]
